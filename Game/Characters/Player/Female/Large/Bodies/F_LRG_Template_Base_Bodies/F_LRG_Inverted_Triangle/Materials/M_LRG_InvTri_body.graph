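# M_LRG_InvTri_body is a Default Lit body material for the large inverted-triangle female player character template.
# Its graph holds only a normal-map sample of LRG_InvTri_normal and two scalar constants, with no color texture or parameters.
# The listing shows no links from these nodes to material outputs.
# Usage flags enable it for skeletal meshes, morph targets and mesh deformers.
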Material /Game/Characters/Player/Female/Large/Bodies/F_LRG_Template_Base_Bodies/F_LRG_Inverted_Triangle/Materials/M_LRG_InvTri_body
graph {
 properties {
  bUsedWithMeshDeformer=true
  bUsedWithMorphTargets=true
  bUsedWithSkeletalMesh=true
  pixelDepthOffsetMode=PDOM_Legacy
  usedShadingModels=Default Lit
 }
 nodes [
  Constant() {r=0.2}
  Constant() {r=0.5}
  TextureSample() {
   samplerType=SAMPLERTYPE_Normal
   texture=&/LRG_InvTri_normal
  }
 ]
}
